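# Niagara system for the Lateral WipeUp emote, flashing short-lived, red-tinted sprites over the character's body with a rising flare.
# Decals pins one sprite per bone/socket of User.SkelMesh by spawn group, while DecalsGlow and Decals_Low use index arrays for subsets.
# Curves animate color, alpha and a dynamic material parameter, with a time-of-day scalar adjusting brightness in most emitters.
# Flare launches upward from below neck_02 with glow sprites and a blue particle light.
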NiagaraSystem /BRCosmetics/Animation/Game/MainPlayer/Emotes/Lateral/FX/NS_Emote_Lateral_WipeUp
dependencies {
 usedDataInterfaces [
  &NiagaraDataInterfaceCurve
  &NiagaraDataInterfaceColorCurve
  &NiagaraDataInterfaceArrayInt32
  &NiagaraDataInterfaceArrayFloat
 ]
 usedDynamicInputs [&FNTimeOfDayScalar]
 usedModules [
  &self:SystemSpawnScript.SystemScriptSource.SystemScriptGraph.NiagaraNodeAssignment_1.SetVariables_#1
  &SystemState
  &self:SystemSpawnScript.SystemScriptSource.SystemScriptGraph.NiagaraNodeAssignment_0.SetVariables_#2
 ]
}
emitters {
 Decals {
  emitterSpawnScript [/SetVariables(Emitter.DepthFade=75, Emitter.VisibleDepthDistance=75) {isSetParameters=true}]
  emitterUpdateScript [
   &EmitterState(Inactive Response=::NewEnumerator1, Life Cycle Mode=::NewEnumerator1, Loop Behavior=::NewEnumerator1, Loop Duration=2, UseLoopDelay=false)
   &SpawnBurst_Instantaneous(Spawn Count=25, Use Spawn Probability=false) {module=SpawnBurst_Instantaneous001}
  ]
  particleSpawnScript [
   &FN_SetSpawnGroup(Number of Spawn Groups=25) {module=FN_SetSpawnGroup001}
   &InitializeParticle(
    Color=<rgba 5 5 5 1>
    Lifetime=0.75
    Sprite Size Mode=::NewEnumerator3
    Uniform Sprite Size=35
    UsePositionOffset=false
    Write Material Random=true
   ) {module=InitializeParticle001}
  ]
  particleUpdateScript [
   &SkeletalMeshLocation(Bone / Socket Index=$Particles.SpawnGroup, Bone Sampling Mode=::NewEnumerator4, MaskSpawnGroup=false, Skeletal Mesh=$User.SkelMesh)
   &ParticleState()
   &CameraOffset(Camera Offset Amount=60)
   &DynamicMaterialParameters(
    Index 0 Param 1=2
    Index 0 Param 2=1
    Index 0 Param 3=&FloatFromCurve(FloatCurve=DataInterface() {propertyValues {Curve {keys [{interpMode=RCIM_Cubic, tangentMode=RCTM_User, time=0, value=-0.25}, {interpMode=RCIM_Cubic, tangentMode=RCTM_User, time=1, value=0.55}]}}})
    Module.Param0WriteEnabled=true
    Module.Param1WriteEnabled=true
    Module.Param2WriteEnabled=true
    Module.Param3WriteEnabled=true
   )
   &ScaleColor(
    Scale Alpha=&FloatFromCurve(FloatCurve=DataInterface() {propertyValues {Curve {keys [
         {arriveTangent=3, interpMode=RCIM_Cubic, leaveTangent=3, time=0, value=0}
         {arriveTangent=1.9398, interpMode=RCIM_Cubic, leaveTangent=1.9398, tangentMode=RCTM_User, time=0.1, value=0.3}
         <at 0.2 1>
         {arriveTangent=-3.5366, interpMode=RCIM_Cubic, leaveTangent=-3.5366, tangentMode=RCTM_User, time=0.4, value=0.95}
         <at 0.7 0>
        ]}}})
    Scale RGB=&MakeVectorFromLinearColorRGB(Color=&ColorFromCurve(ColorCurve=DataInterface() {propertyValues {
        AlphaCurve {keys [<at 0 1 RCIM_Cubic>]}
        BlueCurve {keys [
          {arriveTangent=-0.5763, interpMode=RCIM_Cubic, leaveTangent=-0.5763, time=0, value=0.1967}
          <at 0.3298 0.0067>
          {arriveTangent=0.0131, interpMode=RCIM_Cubic, leaveTangent=0.0131, time=1, value=0.0154}
         ]}
        GreenCurve {keys [{arriveTangent=-0.5362, interpMode=RCIM_Cubic, leaveTangent=-0.5362, time=0, value=0.1768}, <at 0.3298 0>, <at 1 0 RCIM_Cubic>]}
        RedCurve {keys [<at 0 1 RCIM_Cubic>, <at 0.3298 1>, {arriveTangent=-1.0335, interpMode=RCIM_Cubic, leaveTangent=-1.0335, time=1, value=0.3073}]}
       }}))
    ScaleA=true
    ScaleRGB=true
   )
   &ScaleColor(
    Scale RGB=&VectorFromFloat(Value=&FloatFromCurve(FloatCurve=DataInterface() {propertyValues {Curve {keys [
          <at 0 0>
          {arriveTangent=3.0581, interpMode=RCIM_Cubic, leaveTangent=3.0581, tangentMode=RCTM_User, time=0.1, value=0.5}
          {arriveTangent=39.28, interpMode=RCIM_Cubic, leaveTangent=29.7857, tangentMode=RCTM_Break, time=0.2, value=3}
          <at 0.3 20>
          {arriveTangent=-5.1937, interpMode=RCIM_Cubic, leaveTangent=-5.1937, tangentMode=RCTM_User, time=0.4, value=3}
          {arriveTangent=-2.1039, interpMode=RCIM_Cubic, leaveTangent=-2.1039, tangentMode=RCTM_User, time=0.7, value=0.5}
         ]}}}))
    ScaleA=true
    ScaleRGB=true
   ) {module=ScaleColor001}
   &ScaleColor(Scale RGB=&VectorFromFloat(Value=$System.TOD), ScaleA=true, ScaleRGB=true) {module=ScaleColor002}
  ]
  rendererClasses [&NiagaraSpriteRendererProperties]
  renderers [{rendererClass=&NiagaraSpriteRendererProperties}]
 }
 DecalsGlow {
  emitterUpdateScript [
   &EmitterState(Inactive Response=::NewEnumerator1, Life Cycle Mode=::NewEnumerator1, Loop Behavior=::NewEnumerator1, Loop Duration=2.5, UseLoopDelay=false)
   &SpawnBurst_Instantaneous(Spawn Count=3, Spawn Time=0.1, Use Spawn Probability=false) {module=SpawnBurst_Instantaneous001}
  ]
  particleSpawnScript [
   &FN_SetSpawnGroup(Number of Spawn Groups=3) {module=FN_SetSpawnGroup001}
   &InitializeParticle(
    Color=<rgba 5 5 5 1>
    Lifetime=0.75
    Sprite Size Mode=::NewEnumerator3
    Uniform Sprite Size=60
    UsePositionOffset=false
    Write Material Random=true
   ) {module=InitializeParticle001}
  ]
  particleUpdateScript [
   &SkeletalMeshLocation(
    Bone / Socket Index=&SelectIntFromArray(Direct Array Index=&SpawnGroup(), Int Selection Array=DataInterface() {propertyValues {IntData [1,4,16]}})
    Bone Sampling Mode=::NewEnumerator4
    MaskSpawnGroup=false
    Skeletal Mesh=$User.SkelMesh
   )
   &ParticleState()
   &CameraOffset(Camera Offset Amount=10)
   &DynamicMaterialParameters(
    Index 0 Param 1=2
    Index 0 Param 2=1
    Index 0 Param 3=&FloatFromCurve(FloatCurve=DataInterface() {propertyValues {Curve {keys [{interpMode=RCIM_Cubic, tangentMode=RCTM_User, time=0, value=0.525}, {interpMode=RCIM_Cubic, tangentMode=RCTM_User, time=0.8, value=-0.25}]}}})
    Module.Param0WriteEnabled=true
    Module.Param1WriteEnabled=true
    Module.Param2WriteEnabled=true
    Module.Param3WriteEnabled=true
   )
   &ScaleColor(
    Scale Alpha=&FloatFromCurve(
     FloatCurve=DataInterface() {propertyValues {Curve {keys [
         {arriveTangent=-0.3409, interpMode=RCIM_Cubic, leaveTangent=-0.3409, tangentMode=RCTM_User, time=0, value=0}
         <at 0.4 0.625 RCIM_Cubic>
         <at 1 0>
        ]}}}
     Scale Curve=0.05
    )
    Scale RGB=&MakeVectorFromLinearColorRGB(Color=&ColorFromCurve(ColorCurve=DataInterface() {propertyValues {
        AlphaCurve {keys [<at 0 1 RCIM_Cubic>]}
        BlueCurve {keys [<at 0 0>, {arriveTangent=0.0148, interpMode=RCIM_Cubic, leaveTangent=0.0148, time=1, value=0.0148}]}
        GreenCurve {keys [<at 0 0.032>, {arriveTangent=-0.032, interpMode=RCIM_Cubic, leaveTangent=-0.032, time=1, value=0}]}
        RedCurve {keys [<at 0 1>, {arriveTangent=-0.8646, interpMode=RCIM_Cubic, leaveTangent=-0.8646, time=1, value=0.1354}]}
       }}))
    ScaleA=true
    ScaleRGB=true
   )
   &ScaleColor(Scale RGB=&VectorFromFloat(Value=$System.TOD), ScaleA=true, ScaleRGB=true) {module=ScaleColor002}
   &ScaleSpriteSize(Uniform Curve Sprite Scale=DataInterface() {propertyValues {Curve {keys [
        {interpMode=RCIM_Cubic, tangentMode=RCTM_User, time=0, value=0.5}
        {interpMode=RCIM_Cubic, tangentMode=RCTM_User, time=0.2, value=0.95}
        {tangentMode=RCTM_User, time=1, value=0.5}
       ]}}})
   /SetVariables() {isSetParameters=true}
  ]
  rendererClasses [&NiagaraSpriteRendererProperties]
  renderers [{rendererClass=&NiagaraSpriteRendererProperties}]
 }
 Decals_Low {
  emitterUpdateScript [
   &EmitterState(Inactive Response=::NewEnumerator1, Life Cycle Mode=::NewEnumerator1, Loop Behavior=::NewEnumerator1, Loop Duration=1.5, UseLoopDelay=false)
   &SpawnBurst_Instantaneous(Spawn Count=17, Use Spawn Probability=false) {module=SpawnBurst_Instantaneous001}
  ]
  particleSpawnScript [
   &FN_SetSpawnGroup(Number of Spawn Groups=17) {module=FN_SetSpawnGroup001}
   &InitializeParticle(
    Color=<rgba 5 5 5 1>
    Lifetime=0.75
    Sprite Size Mode=::NewEnumerator3
    Uniform Sprite Size=35
    UsePositionOffset=false
    Write Material Random=true
   ) {module=InitializeParticle001}
  ]
  particleUpdateScript [
   &SkeletalMeshLocation(
    Bone / Socket Index=&SelectIntFromArray(
     Direct Array Index=$Particles.SpawnGroup
     Int Selection Array=DataInterface() {propertyValues {IntData [0,1,2,3,6,7,10,11,14,17,18,19,20,21,22,23,24]}}
    )
    Bone Sampling Mode=::NewEnumerator4
    MaskSpawnGroup=false
    Skeletal Mesh=$User.SkelMesh
   )
   &ParticleState()
   &CameraOffset(Camera Offset Amount=60)
   &DynamicMaterialParameters(
    Index 0 Param 1=2
    Index 0 Param 2=1
    Index 0 Param 3=&FloatFromCurve(FloatCurve=DataInterface() {propertyValues {Curve {keys [{interpMode=RCIM_Cubic, tangentMode=RCTM_User, time=0, value=-0.25}, {interpMode=RCIM_Cubic, tangentMode=RCTM_User, time=1, value=0.55}]}}})
    Module.Param0WriteEnabled=true
    Module.Param1WriteEnabled=true
    Module.Param2WriteEnabled=true
    Module.Param3WriteEnabled=true
   )
   &ScaleColor(
    Scale Alpha=&FloatFromCurve(FloatCurve=DataInterface() {propertyValues {Curve {keys [
         {arriveTangent=3, interpMode=RCIM_Cubic, leaveTangent=3, time=0, value=0}
         {arriveTangent=1.9398, interpMode=RCIM_Cubic, leaveTangent=1.9398, tangentMode=RCTM_User, time=0.1, value=0.3}
         <at 0.2 1>
         {arriveTangent=-3.5366, interpMode=RCIM_Cubic, leaveTangent=-3.5366, tangentMode=RCTM_User, time=0.4, value=0.95}
         <at 0.7 0>
        ]}}})
    Scale RGB=&MakeVectorFromLinearColorRGB(Color=&ColorFromCurve(ColorCurve=DataInterface() {propertyValues {
        AlphaCurve {keys [<at 0 1 RCIM_Cubic>]}
        BlueCurve {keys [
          {arriveTangent=-0.5763, interpMode=RCIM_Cubic, leaveTangent=-0.5763, time=0, value=0.1967}
          <at 0.3298 0.0067>
          {arriveTangent=0.0131, interpMode=RCIM_Cubic, leaveTangent=0.0131, time=1, value=0.0154}
         ]}
        GreenCurve {keys [{arriveTangent=-0.5362, interpMode=RCIM_Cubic, leaveTangent=-0.5362, time=0, value=0.1768}, <at 0.3298 0>, <at 1 0 RCIM_Cubic>]}
        RedCurve {keys [<at 0 1 RCIM_Cubic>, <at 0.3298 1>, {arriveTangent=-1.0335, interpMode=RCIM_Cubic, leaveTangent=-1.0335, time=1, value=0.3073}]}
       }}))
    ScaleA=true
    ScaleRGB=true
   )
   &ScaleColor(
    Scale RGB=&VectorFromFloat(Value=&FloatFromCurve(FloatCurve=DataInterface() {propertyValues {Curve {keys [
          <at 0 0>
          {arriveTangent=3.0581, interpMode=RCIM_Cubic, leaveTangent=3.0581, tangentMode=RCTM_User, time=0.1, value=0.5}
          {arriveTangent=39.28, interpMode=RCIM_Cubic, leaveTangent=29.7857, tangentMode=RCTM_Break, time=0.2, value=3}
          <at 0.3 20>
          {arriveTangent=-5.1937, interpMode=RCIM_Cubic, leaveTangent=-5.1937, tangentMode=RCTM_User, time=0.4, value=3}
          {arriveTangent=-2.1039, interpMode=RCIM_Cubic, leaveTangent=-2.1039, tangentMode=RCTM_User, time=0.7, value=0.5}
         ]}}}))
    ScaleA=true
    ScaleRGB=true
   ) {enabled=false, module=ScaleColor001}
   &ScaleColor(Scale RGB=&VectorFromFloat(Value=$System.TOD), ScaleA=true, ScaleRGB=true) {enabled=false, module=ScaleColor002}
   &ScaleSpriteSize(
    Scale Sprite Size Mode=::NewEnumerator0
    Uniform Scale Factor=&SelectFloatFromArray(
     Direct Array Index=$Particles.SpawnGroup
     Float Selection Array=DataInterface() {propertyValues {FloatData [1,1,1,1,1,1,1,1,1,1.25,1.25,1.25,1.25,1.25,1.25,1.25,1.25]}}
    )
   )
  ]
  rendererClasses [&NiagaraSpriteRendererProperties]
  renderers [{rendererClass=&NiagaraSpriteRendererProperties}]
 }
 Flare {
  emitterUpdateScript [
   &EmitterState(Life Cycle Mode=::NewEnumerator1, Loop Behavior=::NewEnumerator1, Loop Duration=2, UseLoopDelay=true)
   &SpawnBurst_Instantaneous(Spawn Time=0.25, Use Spawn Probability=false, UseLoopCountLimit=false)
  ]
  particleSpawnScript [
   &InitializeParticle(
    Color=<rgba 0.2 0.0059 0 1>
    Lifetime=0.2
    Position Mode=::NewEnumerator0
    Sprite Size Mode=::NewEnumerator3
    Uniform Sprite Size=100
    Write Material Random=true
   )
   &SkeletalMeshLocation(
    Bone Sampling Mode=::NewEnumerator4
    MaskSpawnGroup=false
    Sampled Position Offset=<0 0 -70>
    Sampled Position Offset Coordinate Space:ENiagaraSkelMeshTransforms=::NewEnumerator1
    Skeletal Mesh=DataInterface() {propertyValues {
      DefaultMesh=null
      ExcludeBoneName=root
      OptionalComponentClass=null
      PreviewMesh=/BRCosmetics/Characters/Player/Male/Medium/Bodies/M_MED_Parallel_Comic/Meshes/M_MED_Parallel_Comic
      SkinningMode=SkinOnTheFly
      SoftSourceActor=null
      SourceMode=Default
      UvSetIndex=0
      WholeMeshLOD=-1
      bExcludeBone=false
      bReadDeformedGeometry=true
      bRequireCurrentFrameData=true
      ComponentTags []
      FilteredBones [neck_02]
      FilteredSockets []
      MeshUserParameter {parameter {name=None, typeDefHandle {registeredTypeIndex=1}, varData []}}
      SamplingRegions []
     }}
   )
   &AddVelocity(Velocity=<0 0 500>)
   &CameraOffset(Camera Offset Amount=100)
  ]
  particleUpdateScript [
   &SolveForcesAndVelocity(Clamp Velocity=false, Limit Acceleration=false)
   &ParticleState()
   &ScaleSpriteSize(Uniform Curve Sprite Scale=DataInterface() {propertyValues {Curve {keys [
        {arriveTangent=1, interpMode=RCIM_Cubic, leaveTangent=1, time=0, value=0.5}
        {arriveTangent=0.25, interpMode=RCIM_Cubic, leaveTangent=0.25, time=0.5, value=1}
        <at 1 0.75>
       ]}}})
   &ScaleColor(
    Scale Alpha=&FloatFromCurve(
     FloatCurve=DataInterface() {propertyValues {Curve {keys [
         {interpMode=RCIM_Cubic, tangentMode=RCTM_User, time=0, value=0}
         {interpMode=RCIM_Cubic, tangentMode=RCTM_User, time=0.4, value=1}
         {tangentMode=RCTM_User, time=1, value=0}
        ]}}}
     Scale Curve=&Multiply_Float(A=0.5, B=$System.TOD)
    )
    ScaleA=true
    ScaleRGB=true
   )
   /SetVariables(
    Particles.Glow02Color:LinearColor=&MultiplyLinearColor(
     Linear Color=<rgba 1 0 0.0145 0.5>
     Scale Factor=&MakeFloat4FromVectorAndFloat(
      "Vector (RGB)"=<1 1 1>
      "Float (Alpha)"=&Multiply_Float(
       A=&FloatFromCurve(
        FloatCurve=DataInterface() {propertyValues {Curve {keys [
            {interpMode=RCIM_Cubic, tangentMode=RCTM_User, time=0, value=0}
            {interpMode=RCIM_Cubic, tangentMode=RCTM_User, time=0.5, value=1}
            {tangentMode=RCTM_User, time=1, value=0}
           ]}}}
        Scale Curve=0.1
       )
       B=$System.TOD
      )
     )
    )
    Particles.Glow02Size:Vector2f=&Multiply_Vector2D_ByFloat(
     Float=&FloatFromCurve(FloatCurve=DataInterface() {propertyValues {Curve {keys [
          {interpMode=RCIM_Cubic, tangentMode=RCTM_User, time=0, value=0.5}
          {interpMode=RCIM_Cubic, tangentMode=RCTM_User, time=0.5, value=1}
          {tangentMode=RCTM_User, time=1, value=0.5}
         ]}}})
     Vector2D=<400 100>
    )
   ) {isSetParameters=true}
   &FN_Light_Attributes(
    Color=&MultiplyLinearColor(
     Linear Color=<rgba 0 0.5648 1 1>
     Scale Factor=&MakeVector4_FromFloat(Float=&FloatFromCurve(FloatCurve=DataInterface() {propertyValues {Curve {keys [
           {arriveTangent=0.2066, interpMode=RCIM_Cubic, leaveTangent=0.2066, tangentMode=RCTM_User, time=0, value=0}
           {arriveTangent=0.8616, interpMode=RCIM_Cubic, leaveTangent=0.8616, tangentMode=RCTM_User, time=0.3, value=1}
           {arriveTangent=-0.1006, interpMode=RCIM_Cubic, leaveTangent=-0.1006, tangentMode=RCTM_User, time=1, value=0}
          ]}}}))
    )
    Create Color Attribute=true
    Create Exponent Attribute=true
    Create Radius Attribute=true
    Enable Particle Light Enable=true
    Exponent=3
    Radius=125
    Scale Color=5
   )
  ]
  rendererClasses [&NiagaraSpriteRendererProperties, &NiagaraLightRendererProperties]
  renderers [
   {rendererClass=&NiagaraSpriteRendererProperties}
   {rendererClass=&NiagaraSpriteRendererProperties, rendererIndex=1}
   {rendererClass=&NiagaraLightRendererProperties, rendererIndex=2}
  ]
 }
}
summary {
 systemName=NS_Emote_Lateral_WipeUp
 emitters [
  {emitterName=Decals, rendererClasses [&NiagaraSpriteRendererProperties]}
  {emitterName=Flare, rendererClasses [&NiagaraSpriteRendererProperties, &NiagaraLightRendererProperties]}
  {emitterName=DecalsGlow, rendererClasses [&NiagaraSpriteRendererProperties]}
  {emitterName=Decals_Low, rendererClasses [&NiagaraSpriteRendererProperties]}
 ]
 userVariables [User.SkelMesh {
   type=NiagaraDataInterfaceSkeletalMesh
   defaultValue {
    struct=NiagaraExt_VariableValue_DataInterface
    value {dataInterface=&self:NiagaraDataInterfaceSkeletalMesh_0}
   }
  }]
}
